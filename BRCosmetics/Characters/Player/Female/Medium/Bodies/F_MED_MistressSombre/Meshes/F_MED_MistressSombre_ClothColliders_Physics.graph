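# Physics asset of leg colliders for the F_MED_MistressSombre female player body, named as cloth colliders.
# Capsules on both thighs and both feet provide the collision volume, while the calf bodies carry no shapes.
# Four constraints link each foot to its calf and each calf to its thigh.
# Three explicitly free swing and twist, while the left calf-to-thigh joint leaves its motion modes at default.
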PhysicsAsset /BRCosmetics/Characters/Player/Female/Medium/Bodies/F_MED_MistressSombre/Meshes/F_MED_MistressSombre_ClothColliders_Physics
bodies {
 calf_l []
 calf_r []
 foot_l [{shapeType=Capsule, length=14, radius=12.5, shapeName=foot_l_capsule, center=<0 6.6197 0.1519>, rotation=<rot 0 -0.0001 92.6743>}]
 foot_r [{shapeType=Capsule, length=14, radius=12.5, shapeName=foot_l_capsule, center=<0 -6.6197 -0.1519>, rotation=<rot 0 -0.0001 -87.3257>}]
 thigh_l [{shapeType=Capsule, length=40.3746, radius=7.8692, center=<-18.9655 -1.8016 0.543>, rotation=<rot -0.0603 -88.7439 -96.8421>}]
 thigh_r [{shapeType=Capsule, length=40.3746, radius=7.8692, center=<18.9655 1.8016 -0.543>, rotation=<rot -0.0603 -88.7439 83.1579>}]
}
constraints [
 {bone1Name=calf_l, bone2Name=thigh_l, swing1LimitDegrees=45, swing2LimitDegrees=45, twistLimitDegrees=45}
 {
  bone1Name=foot_l
  bone2Name=calf_l
  swing1LimitDegrees=45
  swing1Motion=Free
  swing2LimitDegrees=45
  swing2Motion=Free
  twistLimitDegrees=45
  twistMotion=Free
 }
 {
  bone1Name=foot_r
  bone2Name=calf_r
  swing1LimitDegrees=45
  swing1Motion=Free
  swing2LimitDegrees=45
  swing2Motion=Free
  twistLimitDegrees=45
  twistMotion=Free
 }
 {
  bone1Name=calf_r
  bone2Name=thigh_r
  swing1LimitDegrees=45
  swing1Motion=Free
  swing2LimitDegrees=45
  swing2Motion=Free
  twistLimitDegrees=45
  twistMotion=Free
 }
]
