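# Control Rig for the Mistress Sombre female body cosmetic, adding corrective motion to deform and accessory bones.
# A sequence runs the shared DeformRig function, then its animation-attribute alphas drive kneepad, belt, spike and skirt offsets alongside parent constraints.
# Spherical pose readers store head directions, with backward tilt offsetting the collar roots, and spherical collision runs between collar tips and a head target.
# The crouch-walk remap feeds nothing.
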ControlRigBlueprint /BRCosmetics/Characters/Player/Female/Medium/Bodies/F_MED_MistressSombre/Meshes/F_MED_MistressSombre_CtrlRig
graphs [
 RigVMFunctionLibrary {links [], nodes []}
 RigVMFunction_Sequence_1_ContainedGraph {nodes [
   n1 = RigVMFunctionEntryNode()
   RigVMFunction_Sequence = &RigVMFunction_Sequence(ExecuteContext=@n1.ExecuteContext)
   n2 = &RigVMFunction_Sequence(ExecuteContext=@RigVMFunction_Sequence.B)
   n3 = &RigVMFunction_Sequence(ExecuteContext=@n2.B)
   n4 = &RigVMFunction_Sequence(ExecuteContext=@n3.B)
   n5 = &RigVMFunction_Sequence(ExecuteContext=@n4.B)
   RigVMFunctionReturnNode(A=@RigVMFunction_Sequence.A, B=@n2.A, C=@n3.A, D=@n4.A, E=@n5.A, F=@n5.B)
  ]}
 RigVMModel {nodes [
   RigUnit_BeginExecution = &RigUnit_BeginExecution()
   RigVMFunction_Sequence = RigVMAggregateNode(ExecuteContext=@RigUnit_BeginExecution.ExecutePin)
   FN_DeformRig_FNC = RigVMFunctionReferenceNode(
    DeformRig_ Curve="(deformRig_l_upperarm_up_alpha_72_#1=true,deformRig_r_upperarm_up_alpha_73_#2=true,deformRig_l_upperarm_fr_alpha_74_#3=true,deformRig_r_upperarm_fr_alpha_75_CB51B80C4 ...(truncated)"
    DeformRig_Bool="(upperarm_1_#4=True,elbow_3_#5=True,wrist_7_#6=True,groin_8_#7=True,glute_17_#8=True,knee_9_40C408324CA261 ...(truncated)"
    DeformRig_Debug=(upperarm_debug_18_#4=False,wrist_debug_20_#6=False,lat_debug_26_#9=False,groinGlute_debug_27_#7=False)
    DeformRig_Value="(upperarm_up_18_#10=(Rotation=(X=0,Y=0,Z=0,W=1),Translation=(X=0,Y=0,Z=8),Scale3D=(X=0.7,Y=1,Z=1)),upperarm_fwd_fr_19_#11=(Rotation=(X=0,Y=0,Z=0,W=1),Translation=(X=0,Y=10,Z=0),Scale3D=(X=0.5,Y=1,Z= ...(truncated)"
    ExecuteContext=@RigVMFunction_Sequence.A
   )
   > &RigUnit_ParentConstraint(
    AdvancedSettings=(InterpolationType=Average,RotationOrderForFilter=XZY)
    Child=(Type=Bone,Name="dyn_kneepad_l_01")
    Parents=((Item=(Type=Bone,Name="calf_l"),Weight=0.8),(Item=(Type=Bone,Name="thigh_l"),Weight=0.1))
    Weight=1
   )
   > &RigUnit_ModifyTransforms(
    ItemToModify=((Item=(Type=Bone,Name="dyn_kneepad_l_01"),Transform=(Rotation=(X=0,Y=0,Z=0,W=1),Translation=(X=1,Y=0.2,Z=0),Scale3D=(X=1,Y=1,Z=1))))
    WeightMinimum=0
    Weight=RigVMDispatchNode(BoneName=root, Default=0, Name=deformRig_l_knee_alpha) {template="DISPATCH_RigDispatch_GetAnimAttribute(in Name,in BoneName,in Default,out Value,out Found)"}.Value
   )
   > &RigUnit_ParentConstraint(
    AdvancedSettings=(InterpolationType=Average,RotationOrderForFilter=XZY)
    Child=(Type=Bone,Name="dyn_kneepad_r_01")
    Parents=((Item=(Type=Bone,Name="calf_r"),Weight=0.8),(Item=(Type=Bone,Name="thigh_r"),Weight=0.1))
    Weight=1
   )
   > &RigUnit_ModifyTransforms(
    ItemToModify=((Item=(Type=Bone,Name="dyn_kneepad_r_01"),Transform=(Rotation=(X=0,Y=0,Z=0,W=1),Translation=(X=1,Y=-0.2,Z=0),Scale3D=(X=1,Y=1,Z=1))))
    WeightMinimum=0
    Weight=RigVMDispatchNode(BoneName=root, Default=0, Name=deformRig_r_knee_alpha) {template="DISPATCH_RigDispatch_GetAnimAttribute(in Name,in BoneName,in Default,out Value,out Found)"}.Value
   )
   > &RigUnit_ModifyTransforms(
    ItemToModify=((Item=(Type=Bone,Name="dyn_beltAbis_r_01"),Transform=(Rotation=(X=-0.2578,Y=0.0226,Z=-0.0842,W=0.9623),Translation=(X=1,Y=-4,Z=-4),Scale3D=(X=1,Y=1,Z=1))))
    WeightMinimum=0
    Weight=RigVMDispatchNode(BoneName=root, Default=0, Name=deformRig_r_thigh_up_alpha) {template="DISPATCH_RigDispatch_GetAnimAttribute(in Name,in BoneName,in Default,out Value,out Found)"}.Value
   )
   > &RigUnit_ModifyTransforms(
    ItemToModify=((Item=(Type=Bone,Name="dyn_beltAbis_r_02"),Transform=(Rotation=(X=-0.0941,Y=-0.0789,Z=-0.0941,W=0.988),Translation=(X=-7,Y=-3,Z=-1),Scale3D=(X=1,Y=1,Z=1))))
    WeightMinimum=0
    Weight=RigVMDispatchNode(BoneName=root, Default=0, Name=deformRig_r_thigh_up_alpha) {template="DISPATCH_RigDispatch_GetAnimAttribute(in Name,in BoneName,in Default,out Value,out Found)"}.Value
   )
   n1 = &RigVMFunction_AnimEvalRichCurve(
    Curve=(EditorCurveData=(Keys=((InterpMode=RCIM_Cubic),(InterpMode=RCIM_Cubic,Time=0.4912,Value=0.7938,ArriveTangent=1,LeaveTangent=1),(InterpMode=RCIM_Cubic,Time=1,Value=1))))
    SourceMinimum=0
    TargetMinimum=0
    Value=RigVMDispatchNode(BoneName=root, Default=0, Name=deformRig_r_thigh_up_alpha) {template="DISPATCH_RigDispatch_GetAnimAttribute(in Name,in BoneName,in Default,out Value,out Found)"}.Value
   )
   > &RigUnit_ModifyTransforms(
    ItemToModify=((Item=(Type=Bone,Name="dyn_beltB_r_01"),Transform=(Rotation=(X=0,Y=0,Z=-0.0872,W=0.9962),Translation=(X=-2,Y=-4,Z=0),Scale3D=(X=1,Y=1,Z=1))))
    WeightMinimum=0
    Weight=@n1.Result
   )
   RigVMFunction_AnimEvalRichCurve = &RigVMFunction_AnimEvalRichCurve(
    Curve=(EditorCurveData=(Keys=((InterpMode=RCIM_Cubic),(InterpMode=RCIM_Cubic,Time=0.6255,Value=0.7987,ArriveTangent=1,LeaveTangent=1),(InterpMode=RCIM_Cubic,Time=1,Value=1))))
    SourceMinimum=0
    TargetMinimum=0
    Value=RigVMDispatchNode(BoneName=root, Default=0, Name=deformRig_r_thigh_up_alpha) {template="DISPATCH_RigDispatch_GetAnimAttribute(in Name,in BoneName,in Default,out Value,out Found)"}.Value
   )
   > &RigUnit_ModifyTransforms(
    ItemToModify=((Item=(Type=Bone,Name="dyn_beltD_r_01"),Transform=(Rotation=(X=0,Y=0,Z=-0.0436,W=0.999),Translation=(X=0,Y=0,Z=0),Scale3D=(X=1,Y=1,Z=1))))
    WeightMinimum=0
    Weight=@RigVMFunction_AnimEvalRichCurve.Result
   )
   > &RigUnit_ModifyTransforms(
    ItemToModify=((Item=(Type=Bone,Name="dyn_beltBuckle_l_01"),Transform=(Rotation=(X=0,Y=-0.0872,Z=0,W=0.9962),Translation=(X=-2,Y=-1,Z=2),Scale3D=(X=1,Y=1,Z=0.9))))
    WeightMinimum=0
    Weight=RigVMDispatchNode(BoneName=root, Default=0, Name=deformRig_l_thigh_up_alpha) {template="DISPATCH_RigDispatch_GetAnimAttribute(in Name,in BoneName,in Default,out Value,out Found)"}.Value
   )
   > &RigUnit_ParentConstraint(
    AdvancedSettings=(RotationOrderForFilter=XZY)
    Child=(Type=Bone,Name="dyn_beltSpikeA_r_01")
    Parents="((Item=(Type=Bone,Name=\"spine_02\"),Weight=0.01),(Item=(Type=Bone,Name=\"dyn_skirtA_l_02\"),Weight=0.0179),(Item=(Type=Bone,Name=\"dyn_beltBuckle_l_01\"),Weight=0.0289),(Item=(Type=Bone,Name=\"dyn_skirtA_r_01\"),Weight=0.0624),(Item=(Type=Bone,Name=\"dyn_skirtA_l_ ...(truncated)"
    Weight=1
   )
   > &RigUnit_ParentConstraint(
    AdvancedSettings=(RotationOrderForFilter=XZY)
    Child=(Type=Bone,Name="dyn_beltSpikeB_r_01")
    Parents="((Item=(Type=Bone,Name=\"dyn_skirtA_l_02\"),Weight=0.0072),(Item=(Type=Bone,Name=\"dyn_beltD_r_02\"),Weight=0.0079),(Item=(Type=Bone,Name=\"dyn_skirtA_l_01\"),Weight=0.0301),(Item=(Type=Bone,Name=\"dyn_skirtA_r_01\"),Weight=0.0435),(Item=(Type=Bone,Name=\"pelvis\"), ...(truncated)"
    Weight=1
   )
   > &RigUnit_ParentConstraint(
    AdvancedSettings=(RotationOrderForFilter=XZY)
    Child=(Type=Bone,Name="dyn_beltSpikeC_r_01")
    Parents="((Item=(Type=Bone,Name=\"spine_02\"),Weight=0.0006),(Item=(Type=Bone,Name=\"dyn_skirtA_l_01\"),Weight=0.0025),(Item=(Type=Bone,Name=\"dyn_skirtA_r_01\"),Weight=0.0041),(Item=(Type=Bone,Name=\"pelvis\"),Weight=0.0099),(Item=(Type=Bone,Name=\"dyn_beltD_r_02\"),Weight= ...(truncated)"
    Weight=1
   )
   > &RigUnit_ModifyTransforms(
    ItemToModify=((Item=(Type=Bone,Name="dyn_beltSpikeA_r_01"),Transform=(Rotation=(X=0,Y=0,Z=0,W=1),Translation=(X=-2.5,Y=0,Z=0),Scale3D=(X=1,Y=1,Z=1))))
    WeightMinimum=0
    Weight=RigVMDispatchNode(BoneName=root, Default=0, Name=deformRig_r_thigh_up_alpha) {template="DISPATCH_RigDispatch_GetAnimAttribute(in Name,in BoneName,in Default,out Value,out Found)"}.Value
   )
   > &RigUnit_ModifyTransforms(
    ItemToModify=((Item=(Type=Bone,Name="dyn_beltSpikeB_r_01"),Transform=(Rotation=(X=0,Y=0,Z=0,W=1),Translation=(X=-2,Y=0,Z=0),Scale3D=(X=1,Y=1,Z=1))))
    WeightMinimum=0
    Weight=RigVMDispatchNode(BoneName=root, Default=0, Name=deformRig_r_thigh_up_alpha) {template="DISPATCH_RigDispatch_GetAnimAttribute(in Name,in BoneName,in Default,out Value,out Found)"}.Value
   )
   > &RigUnit_ModifyTransforms(
    ItemToModify=((Item=(Type=Bone,Name="dyn_beltSpikeC_r_01"),Transform=(Rotation=(X=0,Y=0,Z=0,W=1),Translation=(X=-1.5,Y=0,Z=0),Scale3D=(X=1,Y=1,Z=1))))
    WeightMinimum=0
    Weight=RigVMDispatchNode(BoneName=root, Default=0, Name=deformRig_r_thigh_up_alpha) {template="DISPATCH_RigDispatch_GetAnimAttribute(in Name,in BoneName,in Default,out Value,out Found)"}.Value
   )
   n2 = &RigUnit_SphericalPoseReader(
    ActiveRegionScaleFactors=(PositiveWidth=4,NegativeWidth=4,PositiveHeight=2,NegativeHeight=0.1)
    ActiveRegionSize=0.05
    Debug=(bDrawDebug=false,bDraw2D=False,bDrawLocalAxes=False,DebugScale=25,DebugSegments=20,DebugThickness=0.25)
    DriverAxis=(X=1,Y=0,Z=0)
    DriverItem=(Type=Bone,Name="head")
    FalloffRegionScaleFactors=(PositiveWidth=1,NegativeWidth=1,PositiveHeight=1,NegativeHeight=1)
    FalloffSize=0.3
    FlipHeightScaling=False
    FlipWidthScaling=False
    OptionalParentItem=(Type=Bone,Name="neck_01")
    RotationOffset=(X=0,Y=-90,Z=40)
    ExecutePin=@RigVMFunction_Sequence.D
   )
   > RigVMVariableNode(Value=@n2.OutputParam) {variable=head_Fwd}
   > n3 = &RigUnit_SphericalPoseReader(
    ActiveRegionScaleFactors=(PositiveWidth=4,NegativeWidth=4,PositiveHeight=0.1,NegativeHeight=5)
    ActiveRegionSize=0.05
    Debug=(bDrawDebug=true,bDraw2D=False,bDrawLocalAxes=False,DebugScale=25,DebugSegments=20,DebugThickness=0.25)
    DriverAxis=(X=1,Y=0,Z=0)
    DriverItem=(Type=Bone,Name="head")
    FalloffRegionScaleFactors=(PositiveWidth=1.5,NegativeWidth=1.5,PositiveHeight=1,NegativeHeight=1.5)
    FalloffSize=0.3
    FlipHeightScaling=False
    FlipWidthScaling=False
    OptionalParentItem=(Type=Bone,Name="neck_01")
    RotationOffset=(X=0,Y=-90,Z=-50)
   )
   > RigVMVariableNode(Value=@n3.OutputParam) {variable=head_Bck}
   > n4 = &RigUnit_SphericalPoseReader(
    ActiveRegionScaleFactors=(PositiveWidth=0,NegativeWidth=2,PositiveHeight=4,NegativeHeight=4)
    ActiveRegionSize=0.05
    Debug=(bDrawDebug=false,bDraw2D=False,bDrawLocalAxes=false,DebugScale=25,DebugSegments=20,DebugThickness=0.25)
    DriverAxis=(X=1,Y=0,Z=0)
    DriverItem=(Type=Bone,Name="head")
    FalloffRegionScaleFactors=(PositiveWidth=1,NegativeWidth=0.2,PositiveHeight=1,NegativeHeight=1)
    FalloffSize=0.3
    FlipHeightScaling=false
    FlipWidthScaling=false
    OptionalParentItem=(Type=Bone,Name="spine_05")
    RotationOffset=(X=0,Y=-30,Z=-30)
   )
   > RigVMVariableNode(Value=@n4.OutputParam) {variable=head_Rt}
   > n5 = &RigUnit_SphericalPoseReader(
    ActiveRegionScaleFactors=(PositiveWidth=0,NegativeWidth=2,PositiveHeight=4,NegativeHeight=4)
    ActiveRegionSize=0.05
    Debug=(bDrawDebug=false,bDraw2D=False,bDrawLocalAxes=False,DebugScale=25,DebugSegments=20,DebugThickness=0.25)
    DriverAxis=(X=1,Y=0,Z=0)
    DriverItem=(Type=Bone,Name="head")
    FalloffRegionScaleFactors=(PositiveWidth=1,NegativeWidth=0.2,PositiveHeight=1,NegativeHeight=1)
    FalloffSize=0.3
    FlipHeightScaling=false
    FlipWidthScaling=false
    OptionalParentItem=(Type=Bone,Name="spine_05")
    RotationOffset=(X=0,Y=-210,Z=-210)
   )
   > RigVMVariableNode(Value=@n5.OutputParam) {variable=head_Lf}
   > n6 = &RigUnit_SphericalPoseReader(
    ActiveRegionScaleFactors=(PositiveWidth=5,NegativeWidth=5,PositiveHeight=3,NegativeHeight=0.5)
    ActiveRegionSize=0.05
    Debug=(bDrawDebug=false,bDraw2D=False,bDrawLocalAxes=False,DebugScale=25,DebugSegments=20,DebugThickness=0.25)
    DriverAxis=(X=0,Y=1,Z=0)
    DriverItem=(Type=Bone,Name="head")
    FalloffRegionScaleFactors=(PositiveWidth=2,NegativeWidth=2,PositiveHeight=1,NegativeHeight=1.2)
    FalloffSize=0.3
    FlipHeightScaling=false
    FlipWidthScaling=false
    OptionalParentItem=(Type=Bone,Name="spine_05")
    RotationOffset=(X=170,Y=0,Z=10)
   )
   > RigVMVariableNode(Value=@n6.OutputParam) {variable=head_twist_Lf}
   > n7 = &RigUnit_SphericalPoseReader(
    ActiveRegionScaleFactors=(PositiveWidth=5,NegativeWidth=5,PositiveHeight=0.5,NegativeHeight=3)
    ActiveRegionSize=0.05
    Debug=(bDrawDebug=false,bDraw2D=False,bDrawLocalAxes=False,DebugScale=25,DebugSegments=20,DebugThickness=0.25)
    DriverAxis=(X=0,Y=1,Z=0)
    DriverItem=(Type=Bone,Name="head")
    FalloffRegionScaleFactors=(PositiveWidth=2,NegativeWidth=2,PositiveHeight=1.2,NegativeHeight=1)
    FalloffSize=0.3
    FlipHeightScaling=false
    FlipWidthScaling=false
    OptionalParentItem=(Type=Bone,Name="spine_05")
    RotationOffset=(X=50,Y=0,Z=-10)
   )
   > RigVMVariableNode(Value=@n7.OutputParam) {variable=head_twist_Rt}
   n8 = RigVMVariableNode() {variable=head_Bck}
   > &RigUnit_OffsetTransformForItem(
    Item=(Type=Bone,Name="dyn_collarMAINC")
    OffsetTransform=(Rotation=(X=0,Y=0,Z=-0.4226,W=0.9063),Translation=(X=-10,Y=0,Z=5),Scale3D=(X=1,Y=1,Z=1))
    bPropagateToChildren=True
    Weight=@n8.Value
   )
   > &RigUnit_OffsetTransformForItem(
    Item=(Type=Bone,Name="dyn_collarMAINB")
    OffsetTransform=(Rotation=(X=0,Y=0,Z=-0.1736,W=0.9848),Translation=(X=-5,Y=0,Z=0),Scale3D=(X=1,Y=1,Z=1))
    bPropagateToChildren=True
    Weight=@n8.Value
   )
   > &RigUnit_OffsetTransformForItem(
    Item=(Type=Bone,Name="dyn_collarMAINA")
    OffsetTransform=(Rotation=(X=0,Y=0,Z=-0.342,W=0.9397),Translation=(X=-8,Y=0,Z=7),Scale3D=(X=1,Y=1,Z=1))
    bPropagateToChildren=True
    Weight=@n8.Value
   )
   Spherical Collision_2 = RigVMFunctionReferenceNode(
    Debug=true
    Max Radius=20
    Min Radius=19
    Source=(Type=Bone,Name="dyn_collarC_c_03")
    Source Offset=(X=0,Y=0,Z=0)
    Target=(Type=Bone,Name="dyn_head_tgt")
    Target Offset=(X=3,Y=10,Z=-3)
    Weight=1
    ExecuteContext=@RigVMFunction_Sequence.F
   )
   > Spherical Collision_2_1 = RigVMFunctionReferenceNode(
    Debug=false
    Max Radius=20
    Min Radius=19
    Source=(Type=Bone,Name="dyn_collarB_r_03")
    Source Offset=(X=0,Y=0,Z=0)
    Target=(Type=Bone,Name="dyn_head_tgt")
    Target Offset=(X=5,Y=10,Z=0)
    Weight=1
   )
   > Spherical Collision_2_1_1 = RigVMFunctionReferenceNode(
    Debug=false
    Max Radius=18
    Min Radius=18
    Source=(Type=Bone,Name="dyn_collarB_l_03")
    Source Offset=(X=0,Y=0,Z=0)
    Target=(Type=Bone,Name="dyn_head_tgt")
    Target Offset=(X=5,Y=10,Z=0)
    Weight=1
   )
   > Spherical Collision_2_1_2 = RigVMFunctionReferenceNode(
    Debug=false
    Max Radius=18
    Min Radius=18
    Source=(Type=Bone,Name="dyn_collarA_r_03")
    Source Offset=(X=0,Y=0,Z=0)
    Target=(Type=Bone,Name="dyn_head_tgt")
    Target Offset=(X=5,Y=10,Z=0)
    Weight=1
   )
   > Spherical Collision_2_1_1_1 = RigVMFunctionReferenceNode(
    Debug=false
    Max Radius=17
    Min Radius=12
    Source=(Type=Bone,Name="dyn_collarA_l_03")
    Source Offset=(X=0,Y=0,Z=0)
    Target=(Type=Bone,Name="dyn_head_tgt")
    Target Offset=(X=5,Y=10,Z=0)
    Weight=1
   )
   &RigUnit_ParentConstraint(
    AdvancedSettings=(InterpolationType=Average,RotationOrderForFilter=XZY)
    Child=(Type=Bone,Name="dyn_skirtA_l_MSTR")
    Parents=((Item=(Type=Bone,Name="thigh_l"),Weight=1))
    Weight=1
    ExecutePin=@RigVMFunction_Sequence.B
   )
   > &RigUnit_ParentConstraint(
    AdvancedSettings=(InterpolationType=Average,RotationOrderForFilter=XZY)
    Child=(Type=Bone,Name="dyn_skirtA_r_MSTR")
    Parents=((Item=(Type=Bone,Name="thigh_r"),Weight=1))
    Weight=1
   )
   > &RigUnit_ModifyTransforms(
    ItemToModify=((Item=(Type=Bone,Name="dyn_skirtA_l_01"),Transform=(Rotation=(X=-0.383,Y=0.1786,Z=-0.383,W=0.8214),Translation=(X=10,Y=3,Z=4),Scale3D=(X=1,Y=1,Z=1))))
    WeightMinimum=0
    Weight=RigVMDispatchNode(BoneName=root, Default=0, Name=deformRig_l_thigh_up_alpha) {template="DISPATCH_RigDispatch_GetAnimAttribute(in Name,in BoneName,in Default,out Value,out Found)"}.Value
   )
   > &RigUnit_ModifyTransforms(
    ItemToModify=((Item=(Type=Bone,Name="dyn_skirtA_r_02"),Transform=(Rotation=(X=0.2346,Y=0.1094,Z=0.4082,W=0.8754),Translation=(X=5,Y=0,Z=1),Scale3D=(X=1,Y=1,Z=1))))
    WeightMinimum=0
    Weight=RigVMDispatchNode(BoneName=root, Default=0, Name=deformRig_r_thigh_up_alpha) {template="DISPATCH_RigDispatch_GetAnimAttribute(in Name,in BoneName,in Default,out Value,out Found)"}.Value
   )
   Add = &RigVMFunction_MathDoubleAdd(A=@n8.Value, B=RigVMVariableNode() {variable=isCrouchWalking}.Value)
   Remap = &RigVMFunction_MathDoubleRemap(SourceMaximum=2, SourceMinimum=1, TargetMaximum=1, TargetMinimum=0, bClamp=False, Value=@Add.Result)
   RigVMRerouteNode(Value=())
  ]}
]
hierarchy [
 root {loc=<0 0 0>, rot=<0 0 0 1>}
 pelvis {parent=root, loc=<0 -0.5301 98.0571>, rot=<0.0127 0.707 -0.0127 -0.707>}
 spine_01 {parent=pelvis, loc=<6.1208 -0.9328 0>, rot=<0 0 0.034 -0.9994>}
 spine_02 {parent=spine_01, loc=<6.5996 -0.2095 0>, rot=<0 0 0.0238 -0.9997>}
 spine_03 {parent=spine_02, loc=<6.4914 -0.1237 0>, rot=<0 0 0.0149 -0.9999>}
 spine_04 {parent=spine_03, loc=<6.9598 0.0794 0>, rot=<0 0 -0.0132 -0.9999>}
 spine_05 {parent=spine_04, loc=<6.9987 0 0>, rot=<0 0 -0.0015 -1>}
 attach_cape {parent=spine_05, loc=<9.3216 -8.4501 0>, rot=<0.0616 -0.7044 -0.0616 -0.7044>}
 clavicle_l {parent=spine_05, loc=<5.4831 -0.1053 -3.2946>, rot=<0.0465 -0.7289 -0.009 -0.6829>}
 upperarm_l {parent=clavicle_l, loc=<9.8658 -1.2136 -0.4105>, rot=<0.0164 -0.3245 -0.0269 -0.9454>}
 lowerarm_l {parent=upperarm_l, loc=<26.465 -0.8345 0.259>, rot=<0.0036 0.0161 -0.2298 -0.9731>}
 hand_l {parent=lowerarm_l, loc=<22.7355 -0.3203 -0.6105>, rot=<-0.6784 -0.0223 0.0735 -0.7307>}
 thumb_01_l {parent=hand_l, loc=<1.3118 -0.3191 -2.459>, rot=<0.6546 -0.4042 0.0664 -0.6354>}
 thumb_02_l {parent=thumb_01_l, loc=<5.9336 -0.0448 0.0401>, rot=<-0.0037 0.0213 0.298 -0.9543>}
 thumb_03_l {parent=thumb_02_l, loc=<2.2223 0.3682 -0.0907>, rot=<0.0118 -0.013 0.0066 -0.9998>}
 index_metacarpal_l {parent=hand_l, loc=<1.6535 0.0183 -2.96>, rot=<0.0158 -0.0124 -0.0173 -0.9996>}
 index_01_l {parent=index_metacarpal_l, loc=<8.3008 0.0357 -0.0524>, rot=<-0.0044 0.0485 0.0924 -0.9945>}
 index_02_l {parent=index_01_l, loc=<3.3587 -0.3004 0.0452>, rot=<-0.002 0.0027 0.3225 -0.9466>}
 index_03_l {parent=index_02_l, loc=<2.5578 -0.123 -0.0252>, rot=<-0.002 0.0067 0.2558 -0.9667>}
 middle_metacarpal_l {parent=hand_l, loc=<2.1856 -0.1725 -0.5984>, rot=<0.006 0.0168 -0.0329 -0.9993>}
 middle_01_l {parent=middle_metacarpal_l, loc=<7.3275 -0.2447 0.0129>, rot=<-0.0208 0.0042 0.0858 -0.9961>}
 middle_02_l {parent=middle_01_l, loc=<4.0123 0.1401 0.1175>, rot=<-0.0218 0.0258 0.3512 -0.9357>}
 middle_03_l {parent=middle_02_l, loc=<2.6077 0.0534 0.1373>, rot=<0.0082 0.0188 0.1708 -0.9851>}
 ring_metacarpal_l {parent=hand_l, loc=<2.2074 -0.3625 1.3638>, rot=<-0.0284 0.0722 -0.0458 -0.9959>}
 ring_01_l {parent=ring_metacarpal_l, loc=<7.0295 -0.6197 -0.1446>, rot=<-0.0489 -0.0252 0.1892 -0.9804>}
 ring_02_l {parent=ring_01_l, loc=<3.6174 0.4647 -0.1027>, rot=<0 0 0.2806 -0.9598>}
 ring_03_l {parent=ring_02_l, loc=<2.4152 0.114 0.0199>, rot=<0 0 0.1189 -0.9929>}
 pinky_metacarpal_l {parent=hand_l, loc=<1.7616 -0.9753 3.3814>, rot=<-0.0397 0.0814 -0.069 -0.9935>}
 pinky_01_l {parent=pinky_metacarpal_l, loc=<7.1709 -0.4831 -0.1094>, rot=<-0.0654 -0.0035 0.2242 -0.9723>}
 pinky_02_l {parent=pinky_01_l, loc=<2.8274 0.0893 -0.1043>, rot=<0.0037 0.0011 0.2854 -0.9584>}
 pinky_03_l {parent=pinky_02_l, loc=<2.2046 0.0307 -0.0078>, rot=<-0.0002 0.0098 0.1445 -0.9895>}
 deform_wrist_dn_base_l {parent=hand_l, loc=<0 -1.5326 0>, rot=<0 0 0 -1>}
 deform_wrist_dn_l {parent=deform_wrist_dn_base_l, loc=<0 0 0>, rot=<0 0 0 1>}
 deform_wrist_up_base_l {parent=hand_l, loc=<0 2.3009 0>, rot=<0 0 0 -1>}
 deform_wrist_up_l {parent=deform_wrist_up_base_l, loc=<0 0 0>, rot=<0 0 0 1>}
 lowerarm_twist_02_l {parent=lowerarm_l, loc=<10.6707 0 0>, rot=<0 0 0 1>}
 lowerarm_twist_01_l {parent=lowerarm_l, loc=<18.316 0 0>, rot=<0 0 0 1>}
 deform_elbow_in_base_l {parent=lowerarm_l, loc=<0.1834 3.6062 -0.2572>, rot=<0 0 0 -1>}
 deform_elbow_in_l {parent=deform_elbow_in_base_l, loc=<0 0 0>, rot=<0 0 0 1>}
 deform_elbow_out_base_l {parent=lowerarm_l, loc=<0 -4 0>, rot=<0 0 0 1>}
 deform_elbow_out_l {parent=deform_elbow_out_base_l, loc=<0 0 0>, rot=<0 0 0 1>}
 dyn_sleeveA_l_MSTR {parent=lowerarm_l, loc=<4.0711 -2.9421 4.9534>, rot=<0.9555 -0.2652 0.0896 -0.093>}
 dyn_sleeveA_l_07 {parent=dyn_sleeveA_l_MSTR, loc=<0 0 0>, rot=<0 0 0 1>}
 dyn_sleeveB_l_MSTR {parent=lowerarm_l, loc=<-1.9319 -0.3507 -3.882>, rot=<0.1113 -0.147 0.2226 -0.9573>}
 dyn_sleeveB_l_01 {parent=dyn_sleeveB_l_MSTR, loc=<0 0 0>, rot=<0 0 0 1>}
 dyn_sleeveB_l_02 {parent=dyn_sleeveB_l_01, loc=<5.7909 0 0>, rot=<0 0 0 1>}
 dyn_sleeveC_l_MSTR {parent=lowerarm_l, loc=<0.3532 -4.2833 -2.5989>, rot=<0.5569 -0.1809 0.2183 -0.7807>}
 dyn_sleeveC_l_01 {parent=dyn_sleeveC_l_MSTR, loc=<0 0 0>, rot=<0 0 0 1>}
 dyn_sleeveC_l_02 {parent=dyn_sleeveC_l_01, loc=<6.3766 0 0>, rot=<0 0 0 1>}
 dyn_sleeveD_l_MSTR {parent=lowerarm_l, loc=<0.0614 -4.6153 0.5704>, rot=<0.7353 -0.168 0.2239 -0.6172>}
 dyn_sleeveD_l_01 {parent=dyn_sleeveD_l_MSTR, loc=<0 0 0>, rot=<0 0 0 1>}
 dyn_sleeveD_l_02 {parent=dyn_sleeveD_l_01, loc=<7.4533 0 0>, rot=<0 0 0 1>}
 upperarm_twist_01_l {parent=upperarm_l, loc=<7.2135 -0.1134 0>, rot=<0 0 0 1>}
 deform_upperarm_bk_base_l {parent=upperarm_twist_01_l, loc=<-6.4053 -4.3746 0.314>, rot=<0 0 0 -1>}
 deform_upperarm_bk_l {parent=deform_upperarm_bk_base_l, loc=<0 0 0>, rot=<0 0 0 1>}
 deform_upperarm_fr_base_l {parent=upperarm_twist_01_l, loc=<-5.6312 5.092 0.3354>, rot=<0 0 0 -1>}
 deform_upperarm_fr_l {parent=deform_upperarm_fr_base_l, loc=<0 0 0>, rot=<0 0 0 1>}
 deform_upperarm_up_base_l {parent=upperarm_twist_01_l, loc=<-8.891 0.4825 3.9517>, rot=<0 0 0 -1>}
 deform_upperarm_up_l {parent=deform_upperarm_up_base_l, loc=<0 0 0>, rot=<0 0 0 1>}
 upperarm_twist_02_l {parent=upperarm_l, loc=<19.7998 -0.1134 0>, rot=<0 0 0 1>}
 attach_backpack {parent=spine_05, loc=<0 -8.35 0>, rot=<-0.0093 -0.707 0.0093 -0.707>}
 neck_01 {parent=spine_05, loc=<11.3138 0.1529 0>, rot=<0 0 -0.2003 -0.9797>}
 neck_02 {parent=neck_01, loc=<4.5627 0 0>, rot=<0 0 0.0504 -0.9987>}
 head {parent=neck_02, loc=<5.3914 0 0>, rot=<0 0 0.115 -0.9934>}
 dyn_head_tgt {parent=head, loc=<0 0 0>, rot=<0 0 0 1>}
 clavicle_r {parent=spine_05, loc=<5.4831 -0.1053 3.2946>, rot=<0.7289 0.0465 -0.6829 0.009>}
 upperarm_r {parent=clavicle_r, loc=<-9.8658 1.2136 0.4105>, rot=<0.0164 -0.3245 -0.0269 -0.9454>}
 lowerarm_r {parent=upperarm_r, loc=<-26.465 0.8345 -0.259>, rot=<0.0036 0.0161 -0.2298 -0.9731>}
 hand_r {parent=lowerarm_r, loc=<-22.7355 0.3203 0.6106>, rot=<-0.6784 -0.0223 0.0735 -0.7307>}
 thumb_01_r {parent=hand_r, loc=<-1.3118 0.3191 2.459>, rot=<0.6546 -0.4042 0.0664 -0.6354>}
 thumb_02_r {parent=thumb_01_r, loc=<-5.9336 0.0448 -0.0401>, rot=<-0.0037 0.0213 0.298 -0.9543>}
 thumb_03_r {parent=thumb_02_r, loc=<-2.2223 -0.3682 0.0907>, rot=<0.0118 -0.013 0.0066 -0.9998>}
 index_metacarpal_r {parent=hand_r, loc=<-1.6535 -0.0183 2.96>, rot=<0.0158 -0.0124 -0.0173 -0.9996>}
 index_01_r {parent=index_metacarpal_r, loc=<-8.3008 -0.0357 0.0524>, rot=<-0.0044 0.0485 0.0924 -0.9945>}
 index_02_r {parent=index_01_r, loc=<-3.3587 0.3004 -0.0452>, rot=<-0.002 0.0027 0.3225 -0.9466>}
 index_03_r {parent=index_02_r, loc=<-2.5578 0.123 0.0252>, rot=<-0.002 0.0067 0.2558 -0.9667>}
 middle_metacarpal_r {parent=hand_r, loc=<-2.1856 0.1725 0.5984>, rot=<0.006 0.0168 -0.0329 -0.9993>}
 middle_01_r {parent=middle_metacarpal_r, loc=<-7.3275 0.2447 -0.0129>, rot=<-0.0208 0.0042 0.0858 -0.9961>}
 middle_02_r {parent=middle_01_r, loc=<-4.0123 -0.1401 -0.1175>, rot=<-0.0218 0.0258 0.3512 -0.9357>}
 middle_03_r {parent=middle_02_r, loc=<-2.6077 -0.0534 -0.1373>, rot=<0.0082 0.0188 0.1708 -0.9851>}
 ring_metacarpal_r {parent=hand_r, loc=<-2.2074 0.3625 -1.3638>, rot=<-0.0284 0.0722 -0.0458 -0.9959>}
 ring_01_r {parent=ring_metacarpal_r, loc=<-7.0295 0.6197 0.1446>, rot=<-0.0489 -0.0252 0.1892 -0.9804>}
 ring_02_r {parent=ring_01_r, loc=<-3.6174 -0.4647 0.1027>, rot=<0 0 0.2806 -0.9598>}
 ring_03_r {parent=ring_02_r, loc=<-2.4152 -0.114 -0.0199>, rot=<0 0 0.1189 -0.9929>}
 pinky_metacarpal_r {parent=hand_r, loc=<-1.7616 0.9753 -3.3814>, rot=<-0.0397 0.0814 -0.069 -0.9935>}
 pinky_01_r {parent=pinky_metacarpal_r, loc=<-7.1709 0.4831 0.1094>, rot=<-0.0654 -0.0035 0.2242 -0.9723>}
 pinky_02_r {parent=pinky_01_r, loc=<-2.8274 -0.0893 0.1043>, rot=<0.0037 0.0011 0.2854 -0.9584>}
 pinky_03_r {parent=pinky_02_r, loc=<-2.2046 -0.0307 0.0078>, rot=<-0.0002 0.0098 0.1445 -0.9895>}
 deform_wrist_dn_base_r {parent=hand_r, loc=<-0.0001 1.5328 0>, rot=<0 0 0 -1>}
 deform_wrist_dn_r {parent=deform_wrist_dn_base_r, loc=<0 0 0>, rot=<0 0 0 1>}
 deform_wrist_up_base_r {parent=hand_r, loc=<-0.0002 -2.3007 -0.0001>, rot=<0 0 0 -1>}
 deform_wrist_up_r {parent=deform_wrist_up_base_r, loc=<0 0 0>, rot=<0 0 0 1>}
 lowerarm_twist_02_r {parent=lowerarm_r, loc=<-10.6707 0 0>, rot=<0 0 0 -1>}
 lowerarm_twist_01_r {parent=lowerarm_r, loc=<-18.316 0 0>, rot=<0 0 0 -1>}
 deform_elbow_in_base_r {parent=lowerarm_r, loc=<-0.1834 -3.6062 0.2572>, rot=<0 0 0 -1>}
 deform_elbow_in_r {parent=deform_elbow_in_base_r, loc=<0 0 0>, rot=<0 0 0 1>}
 deform_elbow_out_base_r {parent=lowerarm_r, loc=<0 4 -0.0001>, rot=<0 0 0 -1>}
 deform_elbow_out_r {parent=deform_elbow_out_base_r, loc=<0 0 0>, rot=<0 0 0 1>}
 dyn_sleeveA_r_MSTR {parent=lowerarm_r, loc=<-4.0714 2.9422 -4.9531>, rot=<-0.2652 -0.9555 -0.093 -0.0896>}
 dyn_sleeveA_r_07 {parent=dyn_sleeveA_r_MSTR, loc=<0 0 0>, rot=<0 0 0 1>}
 dyn_sleeveD_r_MSTR {parent=lowerarm_r, loc=<-0.0615 4.6154 -0.5702>, rot=<-0.168 -0.7353 -0.6172 -0.2239>}
 dyn_sleeveD_r_01 {parent=dyn_sleeveD_r_MSTR, loc=<0 0 0>, rot=<0 0 0 1>}
 dyn_sleeveD_r_02 {parent=dyn_sleeveD_r_01, loc=<7.4534 0.0001 -0.0001>, rot=<0 0 0 1>}
 dyn_sleeveC_r_MSTR {parent=lowerarm_r, loc=<-0.353 4.2832 2.5986>, rot=<0.1809 0.5569 0.7807 0.2183>}
 dyn_sleeveC_r_01 {parent=dyn_sleeveC_r_MSTR, loc=<0 0 0>, rot=<0 0 0 1>}
 dyn_sleeveC_r_02 {parent=dyn_sleeveC_r_01, loc=<6.3768 0.0001 0>, rot=<0 0 0 1>}
 dyn_sleeveB_r_MSTR {parent=lowerarm_r, loc=<1.9318 0.3507 3.8822>, rot=<0.147 0.1113 0.9573 0.2226>}
 dyn_sleeveB_r_01 {parent=dyn_sleeveB_r_MSTR, loc=<0 0 0>, rot=<0 0 0 1>}
 dyn_sleeveB_r_02 {parent=dyn_sleeveB_r_01, loc=<5.7906 -0.0001 -0.0003>, rot=<0 0 0 1>}
 upperarm_twist_01_r {parent=upperarm_r, loc=<-7.2135 0.1134 0>, rot=<0 0 0 -1>}
 deform_upperarm_bk_base_r {parent=upperarm_twist_01_r, loc=<6.4053 4.3746 -0.314>, rot=<0 0 0 -1>}
 deform_upperarm_bk_r {parent=deform_upperarm_bk_base_r, loc=<0 0 0>, rot=<0 0 0 1>}
 deform_upperarm_fr_base_r {parent=upperarm_twist_01_r, loc=<5.6312 -5.092 -0.3354>, rot=<0 0 0 -1>}
 deform_upperarm_fr_r {parent=deform_upperarm_fr_base_r, loc=<0 0 0>, rot=<0 0 0 1>}
 deform_upperarm_up_base_r {parent=upperarm_twist_01_r, loc=<8.891 -0.4825 -3.9517>, rot=<0 0 0 -1>}
 deform_upperarm_up_r {parent=deform_upperarm_up_base_r, loc=<0 0 0>, rot=<0 0 0 1>}
 upperarm_twist_02_r {parent=upperarm_r, loc=<-19.7998 0.1134 0>, rot=<0 0 0 -1>}
 dyn_collarA_l_MSTR {parent=spine_05, loc=<8.3473 3.8441 -4.5985>, rot=<0.6825 -0.007 -0.066 -0.7279>}
 dyn_collarA_l_01 {parent=dyn_collarA_l_MSTR, loc=<0 0 0>, rot=<0 0 0 1>}
 dyn_collarA_l_02 {parent=dyn_collarA_l_01, loc=<5.5898 0 0>, rot=<0.0104 0.0716 -0.3813 -0.9216>}
 dyn_collarA_l_03 {parent=dyn_collarA_l_02, loc=<5.1246 0 0>, rot=<0.1735 0.0459 -0.0081 -0.9837>}
 dyn_collarA_r_MSTR {parent=spine_05, loc=<8.3478 3.8441 4.5985>, rot=<-0.6757 0.0076 -0.0659 -0.7342>}
 dyn_collarA_r_01 {parent=dyn_collarA_r_MSTR, loc=<0 0 0>, rot=<0 0 0 1>}
 dyn_collarA_r_02 {parent=dyn_collarA_r_01, loc=<5.5893 0 0>, rot=<-0.0219 -0.0693 -0.3817 -0.9214>}
 dyn_collarA_r_03 {parent=dyn_collarA_r_02, loc=<5.1245 0 0>, rot=<-0.1736 0 0 -0.9848>}
 dyn_collarMAINC {parent=spine_05, loc=<6.4697 -2.4247 0>, rot=<0 0 0 1>}
 dyn_collarC_c_MSTR {parent=dyn_collarMAINC, loc=<4.6058 -1.8264 0>, rot=<0.9899 0.1117 0.0028 -0.0871>}
 dyn_collarC_c_01 {parent=dyn_collarC_c_MSTR, loc=<0 0 0>, rot=<0 0 0 1>}
 dyn_collarC_c_02 {parent=dyn_collarC_c_01, loc=<4.2209 0 0>, rot=<0 0 -0.1769 -0.9842>}
 dyn_collarC_c_03 {parent=dyn_collarC_c_02, loc=<8.0592 0.3506 -0.1948>, rot=<0 0 0 1>}
 dyn_collarMAINB {parent=spine_05, loc=<6.4697 -2.4247 0>, rot=<0 0 0 1>}
 dyn_collarB_l_MSTR {parent=dyn_collarMAINB, loc=<4.6585 1.8107 -4.789>, rot=<0.7767 0.1092 -0.0941 -0.6131>}
 dyn_collarB_l_01 {parent=dyn_collarB_l_MSTR, loc=<0 0 0>, rot=<0 0 0 1>}
 dyn_collarB_l_02 {parent=dyn_collarB_l_01, loc=<4.0754 0 0>, rot=<-0.0115 -0.0018 -0.3004 -0.9537>}
 dyn_collarB_l_03 {parent=dyn_collarB_l_02, loc=<6.3949 -0.7424 -1.9729>, rot=<0.0872 0 0 -0.9962>}
 dyn_collarMAINA {parent=spine_05, loc=<6.4697 -2.4247 0>, rot=<0 0 0 1>}
 dyn_collarB_r_MSTR {parent=dyn_collarMAINA, loc=<4.6583 1.8107 4.789>, rot=<-0.7856 -0.1105 -0.0925 -0.6017>}
 dyn_collarB_r_01 {parent=dyn_collarB_r_MSTR, loc=<0 0 0>, rot=<0 0 0 1>}
 dyn_collarB_r_02 {parent=dyn_collarB_r_01, loc=<4.0758 0 0>, rot=<0.0238 -0.003 -0.3004 -0.9535>}
 dyn_collarB_r_03 {parent=dyn_collarB_r_02, loc=<6.395 -0.7359 1.9753>, rot=<-0.0872 0 0 -0.9962>}
 deform_pec_aim_base_l {parent=spine_04, loc=<3.2456 9.3309 -1.2237>, rot=<0.0603 -0.4033 0.3318 -0.8507>}
 deform_pec_aim_l {parent=deform_pec_aim_base_l, loc=<0 0 0>, rot=<0 0 0 1>}
 deform_pec_base_l {parent=deform_pec_aim_l, loc=<9.3665 0 0>, rot=<0 0 0 1>}
 deform_pec_l {parent=deform_pec_base_l, loc=<0 0 0>, rot=<0 0 0 1>}
 deform_pec_aim_base_r {parent=spine_04, loc=<3.2452 9.3309 1.2237>, rot=<-0.4033 -0.0603 0.8507 0.3318>}
 deform_pec_aim_r {parent=deform_pec_aim_base_r, loc=<0 0 0>, rot=<0 0 0 1>}
 deform_pec_base_r {parent=deform_pec_aim_r, loc=<-9.3667 -0.0003 -0.0004>, rot=<0 0 0 1>}
 deform_pec_r {parent=deform_pec_base_r, loc=<0 0 0>, rot=<0 0 0 1>}
 thigh_l {parent=pelvis, loc=<-1.0855 -0.2109 -8.7053>, rot=<0.0009 0.0355 0.0155 -0.9993>}
 calf_l {parent=thigh_l, loc=<-42.5284 -0.0019 -0.7297>, rot=<0.0006 -0.0161 -0.0391 -0.9991>}
 foot_l {parent=calf_l, loc=<-46.7401 -2.5014 -0.592>, rot=<0.0427 -0.0238 0.0353 -0.9982>}
 ball_l {parent=foot_l, loc=<-3.5658 10.8066 -0.1544>, rot=<0.0001 -0.0001 -0.7063 -0.7079>}
 calf_twist_02_l {parent=calf_l, loc=<-19.1546 -1.7943 -0.2638>, rot=<0.0237 0.0095 -0.0035 -0.9997>}
 calf_twist_01_l {parent=calf_l, loc=<-35.8483 -1.9197 -0.5794>, rot=<0.0237 0.0095 -0.0035 -0.9997>}
 deform_knee_in_base_l {parent=calf_l, loc=<0.8031 -5.0231 0.2456>, rot=<0 0 0 -1>}
 deform_knee_in_l {parent=deform_knee_in_base_l, loc=<0 0 0>, rot=<0 0 0 1>}
 deform_knee_out_base_l {parent=calf_l, loc=<0 4 0>, rot=<0 0 0 -1>}
 deform_knee_out_l {parent=deform_knee_out_base_l, loc=<0 0 0>, rot=<0 0 0 1>}
 dyn_kneepad_l_01 {parent=deform_knee_out_base_l, loc=<0 0 0>, rot=<0 0 0 1>}
 dyn_ankle_l_01 {parent=calf_l, loc=<-40.2369 -1.9526 -0.6623>, rot=<0.0427 -0.0238 0.0353 -0.9982>}
 dyn_ball_l_01 {parent=dyn_ankle_l_01, loc=<-11.9755 7.7808 -0.4383>, rot=<0.0001 -0.0001 -0.7063 -0.7079>}
 thigh_twist_01_l {parent=thigh_l, loc=<-27.952 0 0>, rot=<0.0013 0 0 -1>}
 deform_groin_base_l {parent=thigh_l, loc=<0.4801 5.6776 -0.2486>, rot=<0 0 0 -1>}
 deform_groin_l {parent=deform_groin_base_l, loc=<0 0 0>, rot=<0 0 0 1>}
 thigh_r {parent=pelvis, loc=<-1.0855 -0.2109 8.7053>, rot=<0.0355 -0.0009 0.9993 0.0155>}
 calf_r {parent=thigh_r, loc=<42.5284 0.0019 0.7297>, rot=<0.0006 -0.0161 -0.0391 -0.9991>}
 foot_r {parent=calf_r, loc=<46.7401 2.5014 0.592>, rot=<0.0427 -0.0238 0.0353 -0.9982>}
 ball_r {parent=foot_r, loc=<3.5658 -10.8066 0.1544>, rot=<0.0001 -0.0001 -0.7063 -0.7079>}
 calf_twist_02_r {parent=calf_r, loc=<19.1546 1.7943 0.2638>, rot=<0.0237 0.0095 -0.0035 -0.9997>}
 calf_twist_01_r {parent=calf_r, loc=<35.8483 1.9197 0.5794>, rot=<0.0237 0.0095 -0.0035 -0.9997>}
 deform_knee_in_base_r {parent=calf_r, loc=<-0.8031 5.0231 -0.2457>, rot=<0 0 0 -1>}
 deform_knee_in_r {parent=deform_knee_in_base_r, loc=<0 0 0>, rot=<0 0 0 1>}
 deform_knee_out_base_r {parent=calf_r, loc=<0 -4 0>, rot=<0 0 0 -1>}
 deform_knee_out_r {parent=deform_knee_out_base_r, loc=<0 0 0>, rot=<0 0 0 1>}
 dyn_kneepad_r_01 {parent=deform_knee_out_base_r, loc=<0 0 0>, rot=<0 0 0 1>}
 dyn_ankle_r_01 {parent=calf_r, loc=<40.2369 1.9526 0.6623>, rot=<0.0427 -0.0238 0.0353 -0.9982>}
 dyn_ball_r_01 {parent=dyn_ankle_r_01, loc=<11.9755 -7.7808 0.4383>, rot=<0.0001 -0.0001 -0.7063 -0.7079>}
 thigh_twist_01_r {parent=thigh_r, loc=<27.952 0 0>, rot=<0.0013 0 0 -1>}
 deform_groin_base_r {parent=thigh_r, loc=<-0.4801 -5.6776 0.2486>, rot=<0 0 0 -1>}
 deform_groin_r {parent=deform_groin_base_r, loc=<0 0 0>, rot=<0 0 0 1>}
 deform_glute_base_l {parent=pelvis, loc=<-7.3466 -9.0213 -9.1245>, rot=<0.0172 0.031 -0.4477 -0.8935>}
 deform_glute_l {parent=deform_glute_base_l, loc=<0 0 0>, rot=<0 0 0 1>}
 deform_glute_base_r {parent=pelvis, loc=<-7.3464 -9.0213 9.1245>, rot=<0.031 -0.0172 0.8935 -0.4477>}
 deform_glute_r {parent=deform_glute_base_r, loc=<0 0 0>, rot=<0 0 0 1>}
 dyn_skirtA_l_MSTR {parent=pelvis, loc=<9.0139 7.2601 -5.6706>, rot=<0.1155 -0.8309 -0.5228 -0.1511>}
 dyn_skirtA_l_01 {parent=dyn_skirtA_l_MSTR, loc=<0 0 0>, rot=<0 0 0 1>}
 dyn_skirtA_l_02 {parent=dyn_skirtA_l_01, loc=<9.5008 -0.1371 0.0611>, rot=<-0.0781 0.0128 0.0236 -0.9966>}
 dyn_skirtA_l_03 {parent=dyn_skirtA_l_02, loc=<12.7227 0 0>, rot=<-0.1313 -0.0583 -0.0354 -0.989>}
 dyn_skirtA_l_04 {parent=dyn_skirtA_l_03, loc=<12.0206 0 0>, rot=<-0.0968 -0.0665 -0.0197 -0.9929>}
 dyn_skirtA_l_05 {parent=dyn_skirtA_l_04, loc=<15.4561 0 0>, rot=<-0.1383 -0.0278 -0.0039 -0.99>}
 dyn_skirtA_l_06 {parent=dyn_skirtA_l_05, loc=<16.8399 0 0>, rot=<-0.1305 0 0 -0.9914>}
 dyn_skirtB_l_MSTR {parent=pelvis, loc=<8.7547 -2.0415 -11.3972>, rot=<0.0503 -0.9626 0.2118 -0.1612>}
 dyn_skirtB_l_01 {parent=dyn_skirtB_l_MSTR, loc=<0 0 0>, rot=<0 0 0 1>}
 dyn_skirtB_l_02 {parent=dyn_skirtB_l_01, loc=<8.3597 0 0>, rot=<0 0 0 1>}
 dyn_skirtB_l_03 {parent=dyn_skirtB_l_02, loc=<10.3302 0 0>, rot=<0 -0.0948 0 -0.9955>}
 dyn_skirtB_l_04 {parent=dyn_skirtB_l_03, loc=<14.1883 0 0>, rot=<-0.0733 -0.0281 0.0276 -0.9965>}
 dyn_skirtB_l_05 {parent=dyn_skirtB_l_04, loc=<13.6633 0 0>, rot=<0.0005 -0.02 0.025 -0.9995>}
 dyn_skirtB_l_06 {parent=dyn_skirtB_l_05, loc=<16.1207 0 0>, rot=<0 0 0 1>}
 dyn_skirtC_l_MSTR {parent=pelvis, loc=<9.1847 -7.5499 -5.2233>, rot=<0.1183 -0.6891 0.6978 -0.1557>}
 dyn_skirtC_l_01 {parent=dyn_skirtC_l_MSTR, loc=<0 0 0>, rot=<0 0 0 1>}
 dyn_skirtC_l_02 {parent=dyn_skirtC_l_01, loc=<9.003 0 0>, rot=<0.0004 -0.0558 0.0012 -0.9984>}
 dyn_skirtC_l_03 {parent=dyn_skirtC_l_02, loc=<10.4046 0 0>, rot=<-0.0022 -0.088 -0.008 -0.9961>}
 dyn_skirtC_l_04 {parent=dyn_skirtC_l_03, loc=<13.5699 0 0>, rot=<-0.0002 -0.0217 -0.0017 -0.9998>}
 dyn_skirtC_l_05 {parent=dyn_skirtC_l_04, loc=<12.519 0 0>, rot=<0.0002 -0.0083 0.0031 -1>}
 dyn_skirtC_l_06 {parent=dyn_skirtC_l_05, loc=<17.7386 0 0>, rot=<0 0 0 1>}
 dyn_skirtA_r_MSTR {parent=pelvis, loc=<9.0137 7.2601 5.6707>, rot=<0.1518 0.5601 0.8044 -0.127>}
 dyn_skirtA_r_01 {parent=dyn_skirtA_r_MSTR, loc=<0 0 0>, rot=<0 0 0 1>}
 dyn_skirtA_r_02 {parent=dyn_skirtA_r_01, loc=<9.5019 0 0>, rot=<0.081 0.0075 -0.0174 -0.9965>}
 dyn_skirtA_r_03 {parent=dyn_skirtA_r_02, loc=<12.7226 0 0>, rot=<0.1743 -0.0567 0.0379 -0.9823>}
 dyn_skirtA_r_04 {parent=dyn_skirtA_r_03, loc=<12.0206 0 0>, rot=<0.0968 -0.0665 0.0197 -0.9929>}
 dyn_skirtA_r_05 {parent=dyn_skirtA_r_04, loc=<15.4561 0 0>, rot=<0.1113 -0.0277 0.0046 -0.9934>}
 dyn_skirtA_r_06 {parent=dyn_skirtA_r_05, loc=<16.8398 0 0>, rot=<0 0 0 1>}
 dyn_skirtB_r_MSTR {parent=pelvis, loc=<8.7544 -2.0415 11.3972>, rot=<0.1562 -0.1271 0.9774 -0.0641>}
 dyn_skirtB_r_01 {parent=dyn_skirtB_r_MSTR, loc=<0 0 0>, rot=<0 0 0 1>}
 dyn_skirtB_r_02 {parent=dyn_skirtB_r_01, loc=<8.3594 0 0>, rot=<-0.0872 0 0 -0.9962>}
 dyn_skirtB_r_03 {parent=dyn_skirtB_r_02, loc=<10.3301 0 0>, rot=<0.0916 -0.0917 0.0241 -0.9913>}
 dyn_skirtB_r_04 {parent=dyn_skirtB_r_03, loc=<14.1883 0 0>, rot=<0.0581 -0.0326 -0.0221 -0.9975>}
 dyn_skirtB_r_05 {parent=dyn_skirtB_r_04, loc=<13.6634 0 0>, rot=<0.0231 -0.024 -0.0213 -0.9992>}
 dyn_skirtB_r_06 {parent=dyn_skirtB_r_05, loc=<16.1208 0 0>, rot=<0 0 0 1>}
 dyn_skirtC_r_MSTR {parent=pelvis, loc=<9.1843 -7.5499 5.2233>, rot=<0.1557 -0.6978 0.6891 -0.1183>}
 dyn_skirtC_r_01 {parent=dyn_skirtC_r_MSTR, loc=<0 0 0>, rot=<0 0 0 1>}
 dyn_skirtC_r_02 {parent=dyn_skirtC_r_01, loc=<9.0026 0 0>, rot=<-0.0005 -0.0558 -0.0012 -0.9984>}
 dyn_skirtC_r_03 {parent=dyn_skirtC_r_02, loc=<10.4046 0 0>, rot=<0.0018 -0.088 0.008 -0.9961>}
 dyn_skirtC_r_04 {parent=dyn_skirtC_r_03, loc=<13.5699 0 0>, rot=<0.0002 -0.0217 0.0017 -0.9998>}
 dyn_skirtC_r_05 {parent=dyn_skirtC_r_04, loc=<12.5189 0 0>, rot=<-0.0003 -0.0083 -0.0031 -1>}
 dyn_skirtC_r_06 {parent=dyn_skirtC_r_05, loc=<17.7386 0 0>, rot=<0 0 0 1>}
 dyn_beltC_r_01 {parent=pelvis, loc=<0 0 0>, rot=<0.7092 -0.5857 0.2389 0.3113>}
 dyn_beltC_r_02 {parent=dyn_beltC_r_01, loc=<13.2682 0 0>, rot=<0.3359 -0.6222 0.6816 0.188>}
 dyn_beltB_r_01 {parent=pelvis, loc=<0 0 0>, rot=<-0.4696 0.5002 -0.5122 -0.5167>}
 dyn_beltB_r_02 {parent=dyn_beltB_r_01, loc=<16.1206 0 0>, rot=<-0.0303 -0.7065 0.707 -0.0111>}
 dyn_beltD_r_01 {parent=pelvis, loc=<5.8421 -2.984 -7.1354>, rot=<0.5099 -0.6497 0.4072 0.3901>}
 dyn_beltD_r_02 {parent=dyn_beltD_r_01, loc=<17.8695 0 0>, rot=<0 0 0 1>}
 dyn_beltBuckle_l_01 {parent=pelvis, loc=<6.3726 6.8153 -10.519>, rot=<0.4237 0.2459 0.8364 0.2459>}
 dyn_beltAbis_r_01 {parent=pelvis, loc=<6.6038 2.0776 -9.1088>, rot=<0.3517 -0.4406 0.6764 0.4739>}
 dyn_beltAbis_r_02 {parent=dyn_beltAbis_r_01, loc=<20.9783 0 0>, rot=<0 0 0 1>}
 dyn_beltSpikeA_r_01 {parent=pelvis, loc=<0.1933 9.6119 7.3529>, rot=<-0.0177 -0.9853 -0.1699 -0.003>}
 dyn_beltSpikeB_r_01 {parent=pelvis, loc=<-0.2551 7.7374 11.2449>, rot=<-0.0172 -0.9604 -0.2779 -0.005>}
 dyn_beltSpikeC_r_01 {parent=pelvis, loc=<-0.5911 4.3589 14.7622>, rot=<-0.0153 -0.8531 -0.5215 -0.0093>}
]
variables [
 RigUnit_ItemArray_1_Items {
  type="TArray<FRigElementKey>"
  default=((Type=Bone,Name="dyn_collarC_c_01"),(Type=Bone,Name="dyn_collarC_c_02"),(Type=Bone,Name="dyn_collarC_c_03"))
 }
 RigUnit_ItemArray_2_Items {
  type="TArray<FRigElementKey>"
  default=((Type=Control,Name="dyn_collarC_c_01_ctrl"),(Type=Control,Name="dyn_collarC_c_02_ctrl"),(Type=Control,Name="dyn_collarC_c_03_ctrl"))
 }
 SplineIK_Output Spline {type=FControlRigSpline, default=()}
 head_Bck {type=float, default=0}
 head_Fwd {type=float, default=0}
 head_Lf {type=float, default=0}
 head_Rt {type=float, default=0}
 head_twist_Lf {type=float, default=0}
 head_twist_Rt {type=float, default=0}
 isCrouchWalking {type=double, default=0}
]
