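# Physics asset for the F_MED_MistressSombre female body cosmetic, built around skirt dynamics.
# Six skirt chains (A/B/C, left and right) each link six capsules through a tiny master body to the pelvis.
# Segment joints allow narrow swing and zero twist, with C chains loosest and upper right B joints tightest.
# Pelvis, lower-spine and leg capsules provide body shapes, while a huge, offset root capsule approximates a floor.
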PhysicsAsset /BRCosmetics/Characters/Player/Female/Medium/Bodies/F_MED_MistressSombre/Meshes/F_MED_MistressSombre_Physics
bodies {
 calf_l [{shapeType=Capsule, length=34.1324, radius=9.0138, shapeName=calf_l_capsule, center=<-20.2929 0 0>, rotation=<rot 90 0 0>}]
 calf_r [{shapeType=Capsule, length=34.1324, radius=9.0138, shapeName=calf_l_capsule, center=<20.2929 0 0>, rotation=<rot 90 0 180>}]
 dyn_skirtA_l_01 [{shapeType=Capsule, length=13, radius=6.5, shapeName=dyn_skirtA_l_01_capsule, center=<7.1453 0 7.5>, rotation=<rot 0 0 90>}]
 dyn_skirtA_l_02 [{shapeType=Capsule, length=11.5, radius=8, shapeName=dyn_skirtA_l_02_capsule, center=<7.1453 0 7.5>, rotation=<rot 0 0 90>}]
 dyn_skirtA_l_03 [{shapeType=Capsule, length=11, radius=8.5, shapeName=dyn_skirtA_l_03_capsule, center=<7.1453 0 7.5>, rotation=<rot 0 0 90>}]
 dyn_skirtA_l_04 [{shapeType=Capsule, length=11, radius=8.5, shapeName=dyn_skirtA_l_04_capsule, center=<7.1453 0 7.5>, rotation=<rot 0 0 90>}]
 dyn_skirtA_l_05 [{shapeType=Capsule, length=11, radius=8.5, shapeName=dyn_skirtA_l_05_capsule, center=<7.1453 0 7.5>, rotation=<rot 0 0 90>}]
 dyn_skirtA_l_06 [{shapeType=Capsule, length=11, radius=8.5, shapeName=dyn_skirtA_l_06_capsule, center=<7.1453 0 7.5>, rotation=<rot 90 0 0>}]
 dyn_skirtA_l_MSTR [{shapeType=Capsule, length=0.505, radius=0.505, shapeName=dyn_skirtA_l_MSTR_capsule}]
 dyn_skirtA_r_01 [{shapeType=Capsule, length=13.5, radius=6, shapeName=dyn_skirtA_r_01_capsule, center=<7.1453 0 7.5>, rotation=<rot 0 0 90>}]
 dyn_skirtA_r_02 [{shapeType=Capsule, length=12.5, radius=7, shapeName=dyn_skirtA_r_02_capsule, center=<7.1453 0 7.5>, rotation=<rot 0 0 90>}]
 dyn_skirtA_r_03 [{shapeType=Capsule, length=11, radius=8.5, shapeName=dyn_skirtA_r_03_capsule, center=<7.1453 0 7.5>, rotation=<rot 0 0 90>}]
 dyn_skirtA_r_04 [{shapeType=Capsule, length=11, radius=8.5, shapeName=dyn_skirtA_r_04_capsule, center=<7.1453 0 7.5>, rotation=<rot 0 0 90>}]
 dyn_skirtA_r_05 [{shapeType=Capsule, length=11, radius=8.5, shapeName=dyn_skirtA_r_05_capsule, center=<7.1453 0 7.5>, rotation=<rot 0 0 90>}]
 dyn_skirtA_r_06 [{shapeType=Capsule, length=11, radius=8.5, shapeName=dyn_skirtA_r_06_capsule, center=<7.1453 0 7.5>, rotation=<rot 90 0 0>}]
 dyn_skirtA_r_MSTR [{shapeType=Capsule, length=0.505, radius=0.505, shapeName=dyn_skirtA_r_MSTR_capsule}]
 dyn_skirtB_l_01 [{shapeType=Capsule, length=11.5, radius=8, shapeName=dyn_skirtB_l_01_capsule, center=<7.1453 0 7.5>, rotation=<rot 0 0 90>}]
 dyn_skirtB_l_02 [{shapeType=Capsule, length=11.5, radius=8, shapeName=dyn_skirtB_l_02_capsule, center=<7.1453 0 7.5>, rotation=<rot 0 0 90>}]
 dyn_skirtB_l_03 [{shapeType=Capsule, length=11, radius=8.5, shapeName=dyn_skirtB_l_03_capsule, center=<7.1453 0 7.5>, rotation=<rot 0 0 90>}]
 dyn_skirtB_l_04 [{shapeType=Capsule, length=11, radius=8.5, shapeName=dyn_skirtB_l_04_capsule, center=<7.1453 0 7.5>, rotation=<rot 0 0 90>}]
 dyn_skirtB_l_05 [{shapeType=Capsule, length=11, radius=8.5, shapeName=dyn_skirtB_l_05_capsule, center=<7.1453 0 7.5>, rotation=<rot 0 0 90>}]
 dyn_skirtB_l_06 [{shapeType=Capsule, length=10, radius=9.5, shapeName=dyn_skirtB_l_06_capsule, center=<7.1453 0 7.5>, rotation=<rot 0 0 90>}]
 dyn_skirtB_l_MSTR [{shapeType=Capsule, length=0.505, radius=0.505, shapeName=dyn_skirtB_l_MSTR_capsule}]
 dyn_skirtB_r_01 [{shapeType=Capsule, length=12, radius=7.5, shapeName=dyn_skirtB_r_01_capsule, center=<7.1453 0 7.5>, rotation=<rot 0 0 90>}]
 dyn_skirtB_r_02 [{shapeType=Capsule, length=11.5, radius=8, shapeName=dyn_skirtB_r_02_capsule, center=<7.1453 0 7.5>, rotation=<rot 0 0 90>}]
 dyn_skirtB_r_03 [{shapeType=Capsule, length=11, radius=8.5, shapeName=dyn_skirtB_r_03_capsule, center=<7.1453 0 7.5>, rotation=<rot 0 0 90>}]
 dyn_skirtB_r_04 [{shapeType=Capsule, length=11, radius=8.5, shapeName=dyn_skirtB_r_04_capsule, center=<7.1453 0 7.5>, rotation=<rot 0 0 90>}]
 dyn_skirtB_r_05 [{shapeType=Capsule, length=11, radius=8.5, shapeName=dyn_skirtB_r_05_capsule, center=<7.1453 0 7.5>, rotation=<rot 0 0 90>}]
 dyn_skirtB_r_06 [{shapeType=Capsule, length=10, radius=9.5, shapeName=dyn_skirtB_r_06_capsule, center=<7.1453 0 7.5>, rotation=<rot 0 0 90>}]
 dyn_skirtB_r_MSTR [{shapeType=Capsule, length=0.505, radius=0.505, shapeName=dyn_skirtB_r_MSTR_capsule}]
 dyn_skirtC_l_01 [{shapeType=Capsule, length=10.5, radius=9, shapeName=dyn_skirtC_l_01_capsule, center=<7.1453 0 7.5>, rotation=<rot 0 0 90>}]
 dyn_skirtC_l_02 [{shapeType=Capsule, length=10.5, radius=9, shapeName=dyn_skirtC_l_02_capsule, center=<7.1453 0 7.5>, rotation=<rot 0 0 90>}]
 dyn_skirtC_l_03 [{shapeType=Capsule, length=11, radius=8.5, shapeName=dyn_skirtC_l_03_capsule, center=<7.1453 0 7.5>, rotation=<rot 0 0 90>}]
 dyn_skirtC_l_04 [{shapeType=Capsule, length=11, radius=8.5, shapeName=dyn_skirtC_l_04_capsule, center=<7.1453 0 7.5>, rotation=<rot 0 0 90>}]
 dyn_skirtC_l_05 [{shapeType=Capsule, length=11, radius=8.5, shapeName=dyn_skirtC_l_05_capsule, center=<7.1453 0 7.5>, rotation=<rot 0 0 90>}]
 dyn_skirtC_l_06 [{shapeType=Capsule, length=10, radius=9.5, shapeName=dyn_skirtC_l_06_capsule, center=<7.1453 0 7.5>, rotation=<rot 0 0 90>}]
 dyn_skirtC_l_MSTR [{shapeType=Capsule, length=0.505, radius=0.505, shapeName=dyn_skirtC_l_MSTR_capsule}]
 dyn_skirtC_r_01 [{shapeType=Capsule, length=10.5, radius=9, shapeName=dyn_skirtC_r_01_capsule, center=<7.1453 0 7.5>, rotation=<rot 0 0 90>}]
 dyn_skirtC_r_02 [{shapeType=Capsule, length=10.5, radius=9, shapeName=dyn_skirtC_r_02_capsule, center=<7.1453 0 7.5>, rotation=<rot 0 0 90>}]
 dyn_skirtC_r_03 [{shapeType=Capsule, length=11, radius=8.5, shapeName=dyn_skirtC_r_03_capsule, center=<7.1453 0 7.5>, rotation=<rot 0 0 90>}]
 dyn_skirtC_r_04 [{shapeType=Capsule, length=11, radius=8.5, shapeName=dyn_skirtC_r_04_capsule, center=<7.1453 0 7.5>, rotation=<rot 0 0 90>}]
 dyn_skirtC_r_05 [{shapeType=Capsule, length=11, radius=8.5, shapeName=dyn_skirtC_r_05_capsule, center=<7.1453 0 7.5>, rotation=<rot 0 0 90>}]
 dyn_skirtC_r_06 [{shapeType=Capsule, length=10, radius=9.5, shapeName=dyn_skirtC_r_06_capsule, center=<7.1453 0 7.5>, rotation=<rot 0 0 90>}]
 dyn_skirtC_r_MSTR [{shapeType=Capsule, length=0.505, radius=0.505, shapeName=dyn_skirtC_r_MSTR_capsule}]
 foot_l [{shapeType=Capsule, length=9.6592, radius=7.3794, shapeName=foot_l_capsule}]
 foot_r [{shapeType=Capsule, length=9.6582, radius=7.3804, shapeName=foot_r_capsule}]
 pelvis [{shapeType=Capsule, length=12.17, radius=8.2459, shapeName=pelvis_capsule, center=<0 -1.2854 0>}]
 root [{shapeType=Capsule, length=0.505, radius=300, shapeName=root_capsule, center=<0 0 -290.3259>, rotation=<rot 0 -90 0>}]
 spine_01 [{shapeType=Capsule, length=10, radius=8, shapeName=spine_01_capsule}]
 thigh_l [{shapeType=Capsule, length=38.1716, radius=10.1876, shapeName=thigh_l_capsule, center=<-20.2929 0 3.9866>, rotation=<rot 90 0 0>}]
 thigh_r [{shapeType=Capsule, length=38.1716, radius=10.1876, shapeName=thigh_l_capsule, center=<20.2929 0 -3.9866>, rotation=<rot 90 0 180>}]
}
constraints [
 {
  bone1Name=dyn_skirtA_l_MSTR
  bone2Name=pelvis
  swing1LimitDegrees=45
  swing1Motion=Free
  swing2LimitDegrees=45
  swing2Motion=Free
  twistLimitDegrees=45
  twistMotion=Free
 }
 {bone1Name=dyn_skirtA_l_01, bone2Name=dyn_skirtA_l_MSTR, swing1LimitDegrees=2, swing2LimitDegrees=18, twistLimitDegrees=0}
 {bone1Name=dyn_skirtA_l_02, bone2Name=dyn_skirtA_l_01, swing1LimitDegrees=2, swing2LimitDegrees=18, twistLimitDegrees=0}
 {bone1Name=dyn_skirtA_l_03, bone2Name=dyn_skirtA_l_02, swing1LimitDegrees=2, swing2LimitDegrees=18, twistLimitDegrees=0}
 {bone1Name=dyn_skirtA_l_04, bone2Name=dyn_skirtA_l_03, swing1LimitDegrees=2, swing2LimitDegrees=18, twistLimitDegrees=0}
 {bone1Name=dyn_skirtA_l_05, bone2Name=dyn_skirtA_l_04, swing1LimitDegrees=2, swing2LimitDegrees=18, twistLimitDegrees=0}
 {bone1Name=dyn_skirtA_l_06, bone2Name=dyn_skirtA_l_05, swing1LimitDegrees=2, swing2LimitDegrees=18, twistLimitDegrees=0}
 {
  bone1Name=dyn_skirtA_r_MSTR
  bone2Name=pelvis
  swing1LimitDegrees=45
  swing1Motion=Free
  swing2LimitDegrees=45
  swing2Motion=Free
  twistLimitDegrees=45
  twistMotion=Free
 }
 {bone1Name=dyn_skirtA_r_01, bone2Name=dyn_skirtA_r_MSTR, swing1LimitDegrees=2, swing2LimitDegrees=18, twistLimitDegrees=0}
 {bone1Name=dyn_skirtA_r_02, bone2Name=dyn_skirtA_r_01, swing1LimitDegrees=2, swing2LimitDegrees=18, twistLimitDegrees=0}
 {bone1Name=dyn_skirtA_r_03, bone2Name=dyn_skirtA_r_02, swing1LimitDegrees=2, swing2LimitDegrees=18, twistLimitDegrees=0}
 {bone1Name=dyn_skirtA_r_04, bone2Name=dyn_skirtA_r_03, swing1LimitDegrees=2, swing2LimitDegrees=18, twistLimitDegrees=0}
 {bone1Name=dyn_skirtA_r_05, bone2Name=dyn_skirtA_r_04, swing1LimitDegrees=2, swing2LimitDegrees=18, twistLimitDegrees=0}
 {bone1Name=dyn_skirtA_r_06, bone2Name=dyn_skirtA_r_05, swing1LimitDegrees=2, swing2LimitDegrees=18, twistLimitDegrees=0}
 {
  bone1Name=dyn_skirtB_l_MSTR
  bone2Name=pelvis
  swing1LimitDegrees=45
  swing1Motion=Free
  swing2LimitDegrees=45
  swing2Motion=Free
  twistLimitDegrees=45
  twistMotion=Free
 }
 {bone1Name=dyn_skirtB_l_01, bone2Name=dyn_skirtB_l_MSTR, swing1LimitDegrees=2, swing2LimitDegrees=18, twistLimitDegrees=0}
 {bone1Name=dyn_skirtB_l_02, bone2Name=dyn_skirtB_l_01, swing1LimitDegrees=2, swing2LimitDegrees=18, twistLimitDegrees=0}
 {bone1Name=dyn_skirtB_l_03, bone2Name=dyn_skirtB_l_02, swing1LimitDegrees=2, swing2LimitDegrees=18, twistLimitDegrees=0}
 {bone1Name=dyn_skirtB_l_04, bone2Name=dyn_skirtB_l_03, swing1LimitDegrees=2, swing2LimitDegrees=18, twistLimitDegrees=0}
 {bone1Name=dyn_skirtB_l_05, bone2Name=dyn_skirtB_l_04, swing1LimitDegrees=2, swing2LimitDegrees=18, twistLimitDegrees=0}
 {bone1Name=dyn_skirtB_l_06, bone2Name=dyn_skirtB_l_05, swing1LimitDegrees=2, swing2LimitDegrees=18, twistLimitDegrees=0}
 {
  bone1Name=dyn_skirtB_r_MSTR
  bone2Name=pelvis
  swing1LimitDegrees=45
  swing1Motion=Free
  swing2LimitDegrees=45
  swing2Motion=Free
  twistLimitDegrees=45
  twistMotion=Free
 }
 {bone1Name=dyn_skirtB_r_01, bone2Name=dyn_skirtB_r_MSTR, swing1LimitDegrees=2, swing2LimitDegrees=10, twistLimitDegrees=0}
 {bone1Name=dyn_skirtB_r_02, bone2Name=dyn_skirtB_r_01, swing1LimitDegrees=2, swing2LimitDegrees=12, twistLimitDegrees=0}
 {bone1Name=dyn_skirtB_r_03, bone2Name=dyn_skirtB_r_02, swing1LimitDegrees=2, swing2LimitDegrees=18, twistLimitDegrees=0}
 {bone1Name=dyn_skirtB_r_04, bone2Name=dyn_skirtB_r_03, swing1LimitDegrees=2, swing2LimitDegrees=18, twistLimitDegrees=0}
 {bone1Name=dyn_skirtB_r_05, bone2Name=dyn_skirtB_r_04, swing1LimitDegrees=2, swing2LimitDegrees=18, twistLimitDegrees=0}
 {bone1Name=dyn_skirtB_r_06, bone2Name=dyn_skirtB_r_05, swing1LimitDegrees=2, swing2LimitDegrees=18, twistLimitDegrees=0}
 {
  bone1Name=dyn_skirtC_l_MSTR
  bone2Name=pelvis
  swing1LimitDegrees=45
  swing1Motion=Free
  swing2LimitDegrees=45
  swing2Motion=Free
  twistLimitDegrees=45
  twistMotion=Free
 }
 {bone1Name=dyn_skirtC_l_01, bone2Name=dyn_skirtC_l_MSTR, swing1LimitDegrees=2, swing2LimitDegrees=25, twistLimitDegrees=0}
 {bone1Name=dyn_skirtC_l_02, bone2Name=dyn_skirtC_l_01, swing1LimitDegrees=2, swing2LimitDegrees=25, twistLimitDegrees=0}
 {bone1Name=dyn_skirtC_l_03, bone2Name=dyn_skirtC_l_02, swing1LimitDegrees=2, swing2LimitDegrees=25, twistLimitDegrees=0}
 {bone1Name=dyn_skirtC_l_04, bone2Name=dyn_skirtC_l_03, swing1LimitDegrees=2, swing2LimitDegrees=25, twistLimitDegrees=0}
 {bone1Name=dyn_skirtC_l_05, bone2Name=dyn_skirtC_l_04, swing1LimitDegrees=2, swing2LimitDegrees=25, twistLimitDegrees=0}
 {bone1Name=dyn_skirtC_l_06, bone2Name=dyn_skirtC_l_05, swing1LimitDegrees=2, swing2LimitDegrees=25, twistLimitDegrees=0}
 {
  bone1Name=dyn_skirtC_r_MSTR
  bone2Name=pelvis
  swing1LimitDegrees=45
  swing1Motion=Free
  swing2LimitDegrees=45
  swing2Motion=Free
  twistLimitDegrees=45
  twistMotion=Free
 }
 {bone1Name=dyn_skirtC_r_01, bone2Name=dyn_skirtC_r_MSTR, swing1LimitDegrees=2, swing2LimitDegrees=25, twistLimitDegrees=0}
 {bone1Name=dyn_skirtC_r_02, bone2Name=dyn_skirtC_r_01, swing1LimitDegrees=2, swing2LimitDegrees=25, twistLimitDegrees=0}
 {bone1Name=dyn_skirtC_r_03, bone2Name=dyn_skirtC_r_02, swing1LimitDegrees=2, swing2LimitDegrees=25, twistLimitDegrees=0}
 {bone1Name=dyn_skirtC_r_04, bone2Name=dyn_skirtC_r_03, swing1LimitDegrees=2, swing2LimitDegrees=25, twistLimitDegrees=0}
 {bone1Name=dyn_skirtC_r_05, bone2Name=dyn_skirtC_r_04, swing1LimitDegrees=2, swing2LimitDegrees=25, twistLimitDegrees=0}
 {bone1Name=dyn_skirtC_r_06, bone2Name=dyn_skirtC_r_05, swing1LimitDegrees=2, swing2LimitDegrees=25, twistLimitDegrees=0}
 {
  bone1Name=spine_01
  bone2Name=pelvis
  swing1LimitDegrees=45
  swing1Motion=Free
  swing2LimitDegrees=45
  swing2Motion=Free
  twistLimitDegrees=45
  twistMotion=Free
 }
 {
  bone1Name=thigh_l
  bone2Name=pelvis
  swing1LimitDegrees=45
  swing1Motion=Free
  swing2LimitDegrees=45
  swing2Motion=Free
  twistLimitDegrees=45
  twistMotion=Free
 }
 {
  bone1Name=calf_l
  bone2Name=thigh_l
  swing1LimitDegrees=45
  swing1Motion=Free
  swing2LimitDegrees=45
  swing2Motion=Free
  twistLimitDegrees=45
  twistMotion=Free
 }
 {
  bone1Name=foot_r
  bone2Name=calf_r
  swing1LimitDegrees=45
  swing1Motion=Free
  swing2LimitDegrees=45
  swing2Motion=Free
  twistLimitDegrees=45
  twistMotion=Free
 }
 {
  bone1Name=foot_l
  bone2Name=calf_l
  swing1LimitDegrees=45
  swing1Motion=Free
  swing2LimitDegrees=45
  swing2Motion=Free
  twistLimitDegrees=45
  twistMotion=Free
 }
 {
  bone1Name=thigh_r
  bone2Name=pelvis
  swing1LimitDegrees=45
  swing1Motion=Free
  swing2LimitDegrees=45
  swing2Motion=Free
  twistLimitDegrees=45
  twistMotion=Free
 }
 {
  bone1Name=calf_r
  bone2Name=thigh_r
  swing1LimitDegrees=45
  swing1Motion=Free
  swing2LimitDegrees=45
  swing2Motion=Free
  twistLimitDegrees=45
  twistMotion=Free
 }
 {
  bone1Name=pelvis
  bone2Name=root
  swing1LimitDegrees=45
  swing1Motion=Free
  swing2LimitDegrees=45
  swing2Motion=Free
  twistLimitDegrees=45
  twistMotion=Free
 }
]
