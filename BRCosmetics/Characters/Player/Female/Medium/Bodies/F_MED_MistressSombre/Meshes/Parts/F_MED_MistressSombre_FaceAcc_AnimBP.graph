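# Animation Blueprint for the face accessory part of the female MistressSombre cosmetic body, posing its two wing bones.
# It copies the pose from the attached parent mesh, then adds a fixed bone-space pitch to the left and right dyn_wing bones at full alpha.
# Each update derives Alpha Output from the head's twist relative to spine_05 on the part's skeletal mesh, but the AnimGraph does not read that value.
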AnimBlueprint /BRCosmetics/Characters/Player/Female/Medium/Bodies/F_MED_MistressSombre/Meshes/Parts/F_MED_MistressSombre_FaceAcc_AnimBP
target_skeleton=&/F_MED_MistressSombre_FaceAcc_Skeleton
graphs [
 AnimGraph {nodes [
   n1 = "Transform (Modify) Bone - Bone: dyn_wing_l_01"(
    ComponentPose=Local To Component(LocalPose=Copy Pose From Mesh() {bUseAttachedParent=true})
    Alpha=1
    Translation="0,0,0"
    Rotation="10,0,0"
    Scale="1,1,1"
   ) {
    rotationMode=BMM_Additive
    rotationSpace=BCS_BoneSpace
    alphaBoolBlend {blendOption=Linear}
    boneToModify {boneName=dyn_wing_l_01}
    rotation {pitch=10}
   }
   n2 = "Transform (Modify) Bone - Bone: dyn_wing_r_01"(ComponentPose=@n1, Alpha=1, Translation="0,0,0", Rotation="10,0,0", Scale="1,1,1") {
    rotationMode=BMM_Additive
    rotationSpace=BCS_BoneSpace
    alphaBoolBlend {blendOption=Linear}
    boneToModify {boneName=dyn_wing_r_01}
    rotation {pitch=10}
   }
   Output Pose(Result=Component To Local(ComponentPose=@n2)) {layerGroup=None}
  ]}
 EventGraph {nodes [
   Event BlueprintUpdateAnimation()
   > n1 = Sequence()
   > Set NewVar(NewVar:Skeletal Mesh Component Object Reference=GetPartSkeletalMeshComponent()) -> (Output_Get:Skeletal Mesh Component Object Reference)
   Set Alpha Output(
    execute=@n1.then_1
    Alpha Output:float=GetAlphaFromReferenceTransformDelta(
     SkeletalMeshComponent=Get NewVar() -> (NewVar:Skeletal Mesh Component Object Reference)
     BoneName=head
     BaseName=spine_05
     InputRange=(X=0,Y=-30)
     TwistAxis="1,0,0"
    )
   ) -> (Output_Get:float)
  ]}
]
variables [
 Alpha Output {type=number}
 NewVar {type=SkeletalMeshComponent}
]
